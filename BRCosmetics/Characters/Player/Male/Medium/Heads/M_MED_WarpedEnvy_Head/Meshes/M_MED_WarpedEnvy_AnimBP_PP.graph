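# Post-process AnimBP for the WarpedEnvy head that dampens head and neck rotation away from the reference pose.
# With the Venice meta tag, it computes negated, per-axis scaled deltas, varying yaw scale by sign.
# Additive bone transforms apply them to head, neck_02 and neck_01, except during emotes outside the front end.
# The FrontEnd_Idle_Bypass curve blends the neck and head back to the uncorrected, slotted input pose.
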AnimBlueprint /BRCosmetics/Characters/Player/Male/Medium/Heads/M_MED_WarpedEnvy_Head/Meshes/M_MED_WarpedEnvy_AnimBP_PP
target_skeleton=/Game/Characters/Player/Common/Fortnite_Base_Head/Fortnite_Base_3L_Head_Skeleton
graphs [
 AnimGraph {nodes [
   n1 = Layered blend per bone(
    BasePose=Use cached pose 'CACHE_NormalSetup'()
    BlendPoses_0=Slot 'DefaultSlot'(Source=Use cached pose 'BasePose'())
    BlendWeights_0=Get Custom Idle Is Playing Float() -> (Custom Idle Is Playing Float:double)
   ) {layerSetup [{branchFilters [{boneName=neck_01}, {boneName=neck_02}, {boneName=head}]}]}
   Output Pose(Result=@n1)
   n2 = Use cached pose 'BasePose'()
   n3 = Get inGame() -> (inGame:double)
   n4 = "Transform (Modify) Bone - Bone: neck_02"(
    ComponentPose="Transform (Modify) Bone - Bone: head"(
     ComponentPose=Local To Component(LocalPose=@n2)
     Alpha=@n3
     Translation="0,0,0"
     Rotation=Get Head Rot() -> (Head Rot:Rotator)
     Scale="1,1,1"
    ) {rotationMode=BMM_Additive, rotationSpace=BCS_BoneSpace, boneToModify {boneName=head}}
    Alpha=@n3
    Translation="0,0,0"
    Rotation=Get Neck 2 Rot() -> (Neck 2 Rot:Rotator)
    Scale="1,1,1"
   ) {rotationMode=BMM_Additive, rotationSpace=BCS_BoneSpace, boneToModify {boneName=neck_02}}
   n5 = Get Neck 1 Rot() -> (Neck 1 Rot:Rotator)
   Save cached pose 'CACHE_NormalSetup'(Pose=Component To Local(ComponentPose="Transform (Modify) Bone - Bone: neck_01"(ComponentPose=@n4, Alpha=@n3, Translation="0,0,0", Rotation=@n5, Scale="1,1,1") {rotationMode=BMM_Additive, rotationSpace=BCS_BoneSpace, boneToModify {boneName=neck_01}})) {cacheName=CACHE_NormalSetup}
   Save cached pose 'BasePose'(Pose=Input Pose()) {cacheName=BasePose}
  ]}
 EventGraph {nodes [
   Event BlueprintUpdateAnimation()
   > n1 = Sequence()
   > Branch(Condition=CharacterItemDefinitionHasAnyMetaTag(Tags=Make Array("[0]"=(TagName="Cosmetics.MetaTags.Venice"))))
   > n2 = Sequence()
   n3 = GetOwnerSkeletalMeshComponent()
   n4 = BreakRotator(InRot=BreakTransform(InTransform=GetDeltaTransformFromRefPose(self=@n3, BoneName=head, BaseName=neck_02)).Rotation)
   > n5 = Branch(Condition="float < float"(A=@n4.Yaw))
   > n6 = Set z Percentage(z Percentage:double=0.2) -> (Output_Get:double)
   n7 = Get x Percentage() -> (x Percentage:double)
   n8 = Get y Percentage() -> (y Percentage:double)
   n9 = Get z Percentage() -> (z Percentage:double)
   Set Head Rot(
    execute=[@n12.then, @n6.then]
    Head Rot:Rotator=MakeRotator(
     Roll=float * float!(A=@n4, B=-1, C:double=@n7) -> (ReturnValue) {is=title K2Node_PromotableOperator}
     Pitch=float * float(A=@n4.Pitch, B=-1, C=@n8)
     Yaw=float * float(A=@n4.Yaw, B=-1, C=@n9)
    )
   ) -> (Output_Get:Rotator)
   n10 = NOT Boolean(A=Get bIsPlayingEmote())
   n11 = Get bIsInFrontEnd()
   > Set inGame(inGame:double="To Float (Boolean)"(InBool=OR Boolean(A=@n10, B=@n11))) -> (Output_Get:double)
   n12 = Set z Percentage(execute=@n5.else, z Percentage=0)
   n13 = Neck Values!(execute=@n2.then_1, self, Bone Name:Name=neck_02, "Base Name ":Name=neck_01) -> (then, Joint Rotation:Rotator)
   > Set Neck 2 Rot(Neck 2 Rot:Rotator=@n13) -> (Output_Get:Rotator)
   > n14 = Neck Values(Bone Name=neck_01, "Base Name "=neck_01)
   > Set Neck 1 Rot(Neck 1 Rot:Rotator=@n14) -> (Output_Get:Rotator)
   Branch(execute=@n1.then_1, Condition=Get bIsInFrontEnd())
   > Set Custom Idle Is Playing Float(Custom Idle Is Playing Float:double=GetCurveValue(self=GetOwnerFrontEndAnimInstance(), CurveName=FrontEnd_Idle_Bypass)) -> (Output_Get:double)
   TryGetPawnOwner()
   Event OnCharacterCustomizationLoaded()
  ]}
 Neck Values {nodes [
   n1 = Neck Values!() -> (then, Bone Name:Name, "Base Name ":Name) {is=title K2Node_FunctionEntry}
   n2 = GetOwnerSkeletalMeshComponent()
   n3 = BreakRotator(InRot=BreakTransform(InTransform=GetDeltaTransformFromRefPose(self=@n2, BoneName=@n1, BaseName=@n1."Base Name ")).Rotation)
   > Branch(Condition="float < float"(A=@n3.Yaw))
   > n4 = Set z Percentage_neck(z Percentage_neck:double=1.2) -> (Output_Get:double)
   n5 = Get z Percentage_neck() -> (z Percentage_neck:double)
   n6 = Get x Percentage_neck() -> (x Percentage_neck:double)
   n7 = Get y Percentage_neck() -> (y Percentage_neck:double)
   Return Node!(
    execute=[@n8.then, @n4.then]
    Joint Rotation:Rotator=MakeRotator(Roll=float * float(A=@n3, B=-1, C=@n6), Pitch=float * float(A=@n3.Pitch, B=-1, C=@n7), Yaw=float * float(A=@n3.Yaw, B=-1, C=@n5))
   )
   n8 = Set z Percentage_neck(z Percentage_neck=0)
  ]}
]
variables [
 Custom Idle Is Playing Float {type=number}
 Head Rot {type=Rotator}
 inGame {type=number}
 Neck 1 Rot {type=Rotator}
 Neck 2 Rot {type=Rotator}
 x Percentage {type=number, default=0.8}
 x Percentage_neck {type=number, default=0.3}
 y Percentage {type=number, default=0.8}
 y Percentage_neck {type=number, default=0.3}
 z Percentage {type=number, default=1.8}
 z Percentage_neck {type=number}
]
